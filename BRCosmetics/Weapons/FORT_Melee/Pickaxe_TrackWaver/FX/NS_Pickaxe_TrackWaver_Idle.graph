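# Niagara idle effect for the TrackWaver pickaxe, adding blade glows, a point light and drifting motes.
# Two burst glow emitters pin six persistent, aligned sprites to indexed sockets via a shared socket reader, with time-of-day-adjusted colors.
# Motes spawn at socket locations, then fall with curl noise and drag while pulsing and shrinking.
# All emitters select colors by VariantID and blend from warm to blue using User.TW_FXReactvity.
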NiagaraSystem /BRCosmetics/Weapons/FORT_Melee/Pickaxe_TrackWaver/FX/NS_Pickaxe_TrackWaver_Idle
dependencies {
 usedDataInterfaces [
  &NiagaraDataInterfaceArrayInt32
  &NiagaraDataInterfaceArrayFloat3
  &NiagaraDataInterfaceCurve
  &NiagaraDataInterfaceSocketReader
 ]
 usedDynamicInputs [&FNTimeOfDayScalar]
 usedModules [
  &self:SystemSpawnScript.SystemScriptSource.SystemScriptGraph.NiagaraNodeAssignment_1.SetVariables_#1
  &SystemState
 ]
}
emitters {
 NE_BladeGlow_D {
  emitterUpdateScript [
   &EmitterState(
    Inactive Response=::NewEnumerator1
    Life Cycle Mode=::NewEnumerator1
    Loop Behavior=::NewEnumerator1
    Loop Duration Mode=::NewEnumerator1
    UseLoopDelay=false
   )
   &SpawnBurst_Instantaneous(Spawn Count=6, Use Spawn Probability=false, UseLoopCountLimit=false)
  ]
  particleSpawnScript [
   &PartitionParticles(Number Of Partitions=6)
   &InitializeParticle(
    Color Mode=::NewEnumerator0
    Sprite Size=&MergeHelper_Vector2Switch(0=<30 60>, 1=<30 60>, 2=<30 56>, 3=<30 56>, 4=<30 56>, 5=<30 56>, Index=$Particles.PartitionParticles.Partition, Num=::NewEnumerator4)
    Sprite Size Mode=::NewEnumerator4
    UsePositionOffset=false
    Write Material Random=true
   )
   &CameraOffset(Camera Offset Amount=3)
  ]
  particleUpdateScript [
   &ParticleState(Kill Particles When Lifetime Has Elapsed=false, Let Infinitely Lived Particles Die When Emitter Deactivates=true)
   &SocketLocation(
    SamplePosOffset=true
    Sampled Position Offset=&SelectVectorFromArray(
     Direct Array Index=$Particles.PartitionParticles.Partition
     Vector Selection Array=DataInterface() {propertyValues {FloatData [<0 -15 5>, <0 13 5>, <0 -8 0.5>, <0 6 0.5>, <0 -5 0>, <0 3 0>]}}
    )
    Socket Index=&SelectIntFromArray(Direct Array Index=$Particles.PartitionParticles.Partition, Int Selection Array=DataInterface() {propertyValues {IntData [0,1,2,3,4,5]}})
    Socket Reader=$System.Socket Reader
    Socket Sampling Mode=::NewEnumerator3
    VelocitySpeedLimit=false
   )
   &SpriteFacingAndAlignment(
    Sprite Alignment=&SelectVectorFromArray(
     Direct Array Index=$Particles.PartitionParticles.Partition
     Vector Selection Array=DataInterface() {propertyValues {FloatData [<0 1 -1.3>, <0 1 1.3>, <0 1 -0.8>, <0 1 0.8>, <0 1 -0.6>, <0 1 0.6>]}}
    )
    WriteAlignment=true
    WriteFacing=false
   )
   &ScaleColor(
    Color Scale:LinearColor=&MergeHelper_ColorSwitch(
     0=&LerpLinearColors(
      StartColor=&FNTimeOfDayColor(
       Night Color=<rgba 1 0.8333 0 0.05>
       Morning Color=<rgba 1 0.8333 0 0.025>
       Day Color=<rgba 1 0.8333 0 0.05>
       Evening Color=<rgba 1 0.8333 0 0.025>
      )
      EndColor=&FNTimeOfDayColor(
       Night Color=<rgba 0 0.6667 1 0.05>
       Morning Color=<rgba 0 0.6667 1 0.025>
       Day Color=<rgba 0 0.6667 1 0.05>
       Evening Color=<rgba 0 0.6667 1 0.025>
      )
      LerpFactor=&Lerp_Float(A=0, B=1, Alpha=$User.TW_FXReactvity)
     )
     1=&LerpLinearColors(
      StartColor=&FNTimeOfDayColor(
       Night Color=<rgba 1 0.4167 0 0.05>
       Morning Color=<rgba 1 0.4167 0 0.025>
       Day Color=<rgba 1 0.4167 0 0.05>
       Evening Color=<rgba 1 0.4167 0 0.025>
      )
      EndColor=&FNTimeOfDayColor(
       Night Color=<rgba 0 0.6667 1 0.05>
       Morning Color=<rgba 0 0.6667 1 0.025>
       Day Color=<rgba 0 0.6667 1 0.05>
       Evening Color=<rgba 0 0.6667 1 0.025>
      )
      LerpFactor=&Lerp_Float(A=0, B=1, Alpha=$User.TW_FXReactvity)
     )
     2=?
     3=?
     Index=&MakeIntFromFloat(Float=$User.VariantID)
     Num=::NewEnumerator0
    )
    Scale Mode=3
   )
   &ScaleColor(
    Scale Alpha=&MakeCustomFloatFromBool(Bool=$User.bIsFrontEndPreview, True Float=0.35, False Float=0.35)
    ScaleA=true
    ScaleRGB=true
   ) {module=ScaleColor001}
  ]
  rendererClasses [&NiagaraSpriteRendererProperties]
  renderers [{rendererClass=&NiagaraSpriteRendererProperties}]
 }
 NE_BladeGlow_H {
  emitterUpdateScript [
   &EmitterState(
    Inactive Response=::NewEnumerator1
    Life Cycle Mode=::NewEnumerator1
    Loop Behavior=::NewEnumerator1
    Loop Duration Mode=::NewEnumerator1
    UseLoopDelay=false
   )
   &SpawnBurst_Instantaneous(Spawn Count=6, Use Spawn Probability=false, UseLoopCountLimit=false)
  ]
  particleSpawnScript [
   &PartitionParticles(Number Of Partitions=6)
   &InitializeParticle(
    Color Mode=::NewEnumerator0
    Sprite Size=&MergeHelper_Vector2Switch(0=<30 56>, 1=<30 56>, 2=<30 60>, 3=<30 60>, 4=<20 40>, 5=<20 40>, Index=$Particles.PartitionParticles.Partition, Num=::NewEnumerator4)
    Sprite Size Mode=::NewEnumerator4
    UsePositionOffset=false
    Write Material Random=true
   )
   &CameraOffset(Camera Offset Amount=3)
  ]
  particleUpdateScript [
   &ParticleState(Kill Particles When Lifetime Has Elapsed=false, Let Infinitely Lived Particles Die When Emitter Deactivates=true)
   &SocketLocation(
    SamplePosOffset=true
    Sampled Position Offset=&SelectVectorFromArray(
     Direct Array Index=$Particles.PartitionParticles.Partition
     Vector Selection Array=DataInterface() {propertyValues {FloatData [<0 0 0>, <0 0 0>, <0 5 0>, <0 -5 0>, <0 5 1>, <0 -7 1>]}}
    )
    Socket Index=&SelectIntFromArray(Direct Array Index=$Particles.PartitionParticles.Partition, Int Selection Array=DataInterface() {propertyValues {IntData [0,1,2,3,4,5]}})
    Socket Reader=$System.Socket Reader
    Socket Sampling Mode=::NewEnumerator3
    VelocitySpeedLimit=false
   )
   &SpriteFacingAndAlignment(
    Sprite Alignment=&SelectVectorFromArray(
     Direct Array Index=$Particles.PartitionParticles.Partition
     Vector Selection Array=DataInterface() {propertyValues {FloatData [<0 1 0>, <0 1 0>, <0 1 0>, <0 1 0>, <0 1 0.5>, <0 1 -0.5>]}}
    )
    WriteAlignment=true
    WriteFacing=false
   )
   &ScaleColor(
    Color Scale:LinearColor=&MergeHelper_ColorSwitch(
     0=&LerpLinearColors(
      StartColor=&FNTimeOfDayColor(Night Color=<rgba 1 0.8333 0 0.2>, Morning Color=<rgba 1 0.8333 0 0.1>, Day Color=<rgba 1 0.8333 0 0.2>, Evening Color=<rgba 1 0.8333 0 0.1>)
      EndColor=&FNTimeOfDayColor(Night Color=<rgba 0 0.6667 1 0.2>, Morning Color=<rgba 0 0.6667 1 0.1>, Day Color=<rgba 0 0.6667 1 0.2>, Evening Color=<rgba 0 0.6667 1 0.1>)
      LerpFactor=&Lerp_Float(A=0, B=1, Alpha=$User.TW_FXReactvity)
     )
     1=&LerpLinearColors(
      StartColor=&FNTimeOfDayColor(Night Color=<rgba 1 0.4167 0 0.2>, Morning Color=<rgba 1 0.4167 0 0.1>, Day Color=<rgba 1 0.4167 0 0.2>, Evening Color=<rgba 1 0.4167 0 0.1>)
      EndColor=&FNTimeOfDayColor(Night Color=<rgba 0 0.6667 1 0.2>, Morning Color=<rgba 0 0.6667 1 0.1>, Day Color=<rgba 0 0.6667 1 0.2>, Evening Color=<rgba 0 0.6667 1 0.1>)
      LerpFactor=&Lerp_Float(A=0, B=1, Alpha=$User.TW_FXReactvity)
     )
     2=?
     3=?
     Index=&MakeIntFromFloat(Float=$User.VariantID)
     Num=::NewEnumerator0
    )
    Scale Mode=3
   )
   &ScaleColor(
    Scale Alpha=&MakeCustomFloatFromBool(Bool=$User.bIsFrontEndPreview, True Float=0.35, False Float=0.35)
    ScaleA=true
    ScaleRGB=true
   ) {module=ScaleColor001}
  ]
  rendererClasses [&NiagaraSpriteRendererProperties]
  renderers [{rendererClass=&NiagaraSpriteRendererProperties}]
 }
 NE_Light {
  emitterUpdateScript [
   &EmitterState(Life Cycle Mode=::NewEnumerator1, Loop Delay=1, Loop Duration=0.1, UseLoopDelay=false)
   &SpawnBurst_Instantaneous(Spawn Probability=HlslExpression() {hlslExpression="Emitter.LoopCount == 0"}, Use Spawn Probability=true, UseLoopCountLimit=false)
  ]
  particleSpawnScript [&InitializeParticle(Color Mode=::NewEnumerator0, UsePositionOffset=false, Write Material Random=true)]
  particleUpdateScript [
   &ParticleState(Kill Particles When Lifetime Has Elapsed=false, Let Infinitely Lived Particles Die When Emitter Deactivates=true)
   &FN_Light_Attributes(
    Color=&MergeHelper_ColorSwitch(
     0=&LerpLinearColors(
      StartColor=<rgba 50 43.3333 10 1>
      EndColor=<rgba 10 36.6667 50 1>
      LerpFactor=&Lerp_Float(A=0, B=1, Alpha=$User.TW_FXReactvity)
     )
     1=&LerpLinearColors(
      StartColor=<rgba 50 26.6667 10 1>
      EndColor=<rgba 10 36.6667 50 1>
      LerpFactor=&Lerp_Float(A=0, B=1, Alpha=$User.TW_FXReactvity)
     )
     2=?
     3=?
     Index=&MakeIntFromFloat(Float=$User.VariantID)
     Num=::NewEnumerator0
    )
    Create Color Attribute=true
    Create Exponent Attribute=true
    Create Position Attribute=true
    Create Radius Attribute=true
    Exponent=2
    Position=&AddVectorToPosition(Position=$Particles.Position, Vector=<0 0 50>)
    Radius=40
    Scale Color=0.5
   )
  ]
  rendererClasses [&NiagaraLightRendererProperties]
  renderers [{rendererClass=&NiagaraLightRendererProperties, rendererIndex=0}]
 }
 NE_Motes {
  emitterUpdateScript [
   &EmitterState(Inactive Response=::NewEnumerator1, Life Cycle Mode=::NewEnumerator1, Loop Duration=1, UseLoopDelay=false)
   &SpawnRate(SpawnRate=2, Use Spawn Probability=true)
  ]
  particleSpawnScript [
   &InitializeParticle(
    Color=&MergeHelper_ColorSwitch(
     0=&LerpLinearColors(
      StartColor=<rgba 25 12.0345 5 1>
      EndColor=<rgba 5 18.3333 25 1>
      LerpFactor=&Lerp_Float(A=0, B=1, Alpha=$User.TW_FXReactvity)
     )
     1=&LerpLinearColors(
      StartColor=<rgba 25 13.3333 5 1>
      EndColor=<rgba 5 18.3333 25 1>
      LerpFactor=&Lerp_Float(A=0, B=1, Alpha=$User.TW_FXReactvity)
     )
     2=?
     3=?
     Index=&MakeIntFromFloat(Float=$User.VariantID)
     Num=::NewEnumerator0
    )
    Lifetime Mode=::NewEnumerator1
    Sprite Size Mode=::NewEnumerator1
    Uniform Sprite Size Max=3.5
    Uniform Sprite Size Min=1.5
    UsePositionOffset=false
    Write Material Random=true
   )
   &SocketLocation(
    OverrideRandomSeed=false
    SamplePosOffset=false
    Socket Reader=$System.Socket Reader
    Socket Sampling Mode=::NewEnumerator1
    VelocitySpeedLimit=false
   )
  ]
  particleUpdateScript [
   &ParticleState()
   &ScaleSpriteSize(Uniform Curve Sprite Scale=DataInterface() {propertyValues {Curve {defaultValue=0, keys [{tangentMode=RCTM_User, time=0, value=1}, {tangentMode=RCTM_User, time=1, value=0}]}}})
   &Drag(Drag=&RandomRangeFloat(Minimum=5, Maximum=7), Use Linear Drag=true, Use Rotational Drag=true)
   &CurlNoiseForce(Bias Noise=false, Noise Strength=&RandomRangeFloat(Minimum=150, Maximum=300), Pan Noise=false)
   &GravityForce(Coordinate Space=::Local, Gravity=&RandomRangeVector(Minimum=<0 0 -100>, Maximum=<0 0 -200>))
   &Drag(Drag=&RandomRangeFloat(Minimum=3, Maximum=5), Use Linear Drag=true, Use Rotational Drag=true) {module=Drag001}
   &SolveForcesAndVelocity(Clamp Velocity=false, Limit Acceleration=false)
   &ScaleColor(
    Scale Alpha=&FloatFromCurve(FloatCurve=DataInterface() {propertyValues {Curve {keys [{arriveTangent=-3.2059, interpMode=RCIM_Cubic, leaveTangent=-3.2059, tangentMode=RCTM_User, time=0, value=1}, <at 1 0>]}}})
    Scale RGB=&VectorFromFloat(Value=&FloatFromCurve(
      FloatCurve=DataInterface() {propertyValues {Curve {keys [
          {interpMode=RCIM_Cubic, tangentMode=RCTM_User, time=0, value=0}
          {interpMode=RCIM_Cubic, tangentMode=RCTM_User, time=0.2, value=1}
          {interpMode=RCIM_Cubic, tangentMode=RCTM_User, time=0.4, value=0}
          {interpMode=RCIM_Cubic, tangentMode=RCTM_User, time=0.6, value=1}
          {interpMode=RCIM_Cubic, tangentMode=RCTM_User, time=1, value=0}
         ]}}}
      Scale Curve=2
     ))
    ScaleA=true
    ScaleRGB=true
   )
  ]
  rendererClasses [&NiagaraSpriteRendererProperties]
  renderers [{rendererClass=&NiagaraSpriteRendererProperties}]
 }
}
summary {
 systemName=NS_Pickaxe_TrackWaver_Idle
 emitters [
  {emitterName=NE_BladeGlow_H, rendererClasses [&NiagaraSpriteRendererProperties]}
  {emitterName=NE_BladeGlow_D, rendererClasses [&NiagaraSpriteRendererProperties]}
  {emitterName=NE_Light, rendererClasses [&NiagaraLightRendererProperties]}
  {emitterName=NE_Motes, rendererClasses [&NiagaraSpriteRendererProperties]}
 ]
 userVariables [
  User.bIsFrontEnd {type=NiagaraBool}
  User.bIsFrontEndPreview {type=NiagaraBool}
  User.bisSwinging {type=NiagaraBool}
  User.IsFlying {type=NiagaraFloat, defaultValue=0, description=""}
  User.PhysicalSurface {type=NiagaraInt32}
  User.RibbonIDOffset {type=NiagaraInt32}
  User.SwingSectionIndex {type=NiagaraInt32}
  User.TW_FXReactvity {type=NiagaraFloat, defaultValue=0, description=""}
  User.VariantID {type=NiagaraFloat, defaultValue=0, description=""}
 ]
}
