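# Parent material for the TrackWaver pickaxe, adding a parameter-driven reactive color effect to the 2023 character parent material stack.
# Between the PreFX and PostFX functions, it remaps emissive luminance to blend two TrackWaver colors, multiplied by the Tech Art Mask green channel.
# The TW_Reactvity scalar blends the original emissive toward that color.
# The resulting emissive, slightly dimmed, also replaces base color within the masked area.
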Material /BRCosmetics/Weapons/FORT_Melee/Pickaxe_TrackWaver/Materials/M_TrackWaver_Pickaxe_Parent
graph {
 properties {
  bAutomaticallySetUsageInEditor=false
  bUseMaterialAttributes=true
  bUsedWithMeshDeformer=true
  bUsedWithMorphTargets=true
  bUsedWithSkeletalMesh=true
  pixelDepthOffsetMode=PDOM_Legacy
  shadingModel=MSM_FromMaterialExpression
  usedShadingModels=Default Lit | Subsurface | Clear Coat | Subsurface Profile | Cloth
 }
 nodes [
  n1 = MaterialFunctionCall() {materialFunction=&MF_CharacterParent_2023_PreFX}
  MaterialFunctionCall(Material=NamedRerouteUsage(), M=@n1.M, Pre Skinned Local Position=@n1.M, Distance Blend=@n1.M, Fresnel=@n1.M) {materialFunction=&MF_CharacterParent_2023_PostFX}
  n2 = GetMaterialAttributes(None=@n1.Result) {attributeGetTypes [#1, #2]}
  n3 = VectorParameter() {group=TrackWaver, parameterName=TW_ReactiveColor02, defaultValue=<rgba 0 0.1628 0.3739 1>}
  n4 = VectorParameter() {group=TrackWaver, parameterName=TW_ReactiveColor, defaultValue=<rgba 0.2104 1.2075 2.5 1>}
  n5 = SmoothStep(Min=Constant() {r=0.2}, Max=Constant() {r=1}, Value=Desaturation(None=@n2.Emissive Color, Fraction=Constant() {r=1})) {constMax=0.5, constMin=0.2}
  n6 = ComponentMask(None=@n1.Tech Art Mask) {g=true}
  n7 = LinearInterpolate(
   A=@n2.Emissive Color
   B=Multiply(A=LinearInterpolate(A=@n3.RGB, B=@n4.RGB, Alpha=@n5), B=@n6)
   Alpha=ScalarParameter() {group=TrackWaver, parameterName=TW_Reactvity}
  )
  n8 = NamedRerouteUsage()
  n9 = Multiply(A=@n7) {constB=0.7}
  NamedRerouteDeclaration(None=SetMaterialAttributes!(MaterialAttributes=@n2.MaterialAttributes, Emissive Color=@n7, Base Color=LinearInterpolate(A=@n8, B=@n9, Alpha=@n6)) {attributeSetTypes [#1, #2]}) {name=MA_OUT, nodeColor=<rgba 0 0.0391 1 1>}
  NamedRerouteDeclaration(None=@n2.Base Color) {name=BaseColor_IN, nodeColor=<rgba 1 0 0.6797 1>}
 ]
}
